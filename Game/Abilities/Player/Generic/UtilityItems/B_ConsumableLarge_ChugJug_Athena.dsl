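; Blueprint for the large Chug Jug consumable in Athena, which appears to carry only debug logic on top of its parent class.
; On BeginPlay it calls the parent, then reads the FX.FlipSimulation ABTest and Version console variables and prints them as an "ABTest/Version" string.
; The construction script only defers to the parent, and the Tick event is a disabled placeholder.
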(event EventBeginPlay
  (|CallParentFunction)
  (bind _returnvalue (Development|GetConsoleVariableIntValue "FX.FlipSimulation.Version"))
  (bind _returnvalue_2 (Development|GetConsoleVariableIntValue "FX.FlipSimulation.ABTest"))
  (Development|PrintString (Utilities|String|Append (Utilities|String|ToString(Integer) _returnvalue_2) "/" (Utilities|String|ToString(Integer) _returnvalue))))

(event EventTick
  (params (DeltaSeconds :type Float%20%28single-precision%29))
  (|CallParentFunction DeltaSeconds) ; This node is disabled and will not be called.\nDrag off pins to build functionality.
)

(fn ConstructionScript
  (|CallParentFunction))
