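; Blueprint for the small Mini Shield consumable item, controlling the Niagara liquid effect shown while it is drunk.
; Consume Liquid FX passes the drink duration to the Liquid component and sets its IsDrinking flag.
; Reset Liquid FX clears IsDrinking and writes IsFinished as the inverse of its IsFinished input.
; Both events print debug strings, and BeginPlay, Tick and the construction script add no logic of their own.
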(event EventBeginPlay
  (|CallParentFunction) ; This node is disabled and will not be called.\nDrag off pins to build functionality.
)

(event EventTick
  (params (DeltaSeconds :type Float%20%28single-precision%29))
  (|CallParentFunction DeltaSeconds) ; This node is disabled and will not be called.\nDrag off pins to build functionality.
)

(event Custom|Consume%20Liquid%20FX
  (params (Duration :type Float%20%28double-precision%29))
  (bind _liquid (Variables|Default|GetLiquid))
  (Niagara|SetNiagaraVariable(Float) _liquid "Duration" Duration)
  (Niagara|SetNiagaraVariable(Bool) _liquid "IsDrinking" true)
  (Development|PrintString "Drinking" true true "(R=0.000000,G=1.000000,B=0.005721,A=1.000000)"))

(event Custom|Reset%20Liquid%20FX
  (params (IsFinished :type Boolean))
  (bind _liquid (Variables|Default|GetLiquid))
  (Niagara|SetNiagaraVariable(Bool) _liquid "IsDrinking" false)
  (Niagara|SetNiagaraVariable(Bool) _liquid "IsFinished" (not IsFinished))
  (Development|PrintString "Stop Drinking" true true "(R=1.000000,G=0.000000,B=0.010551,A=1.000000)"))

(fn ConstructionScript
  (|CallParentFunction))
