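; Weapon Blueprint for the Molotov grenade utility item in Athena, handling its ambient sound while equipped.
; On equip and attach it calls the parent event, then fades the AmbientAudio component in over a short linear ramp.
; On local unequip it calls the parent event, then fades the same component out.
; The construction script only calls its parent, so this graph adds no throw or damage logic.
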(event FortWeapon|EventOnEquippedAndAttached
  ;[1 1 1] On Equip, turn on the Ambient Audio - On Unequip, turn it off.
  (|CallParentFunction)
  (Audio|Components|Audio|FadeIn (Variables|Default|GetAmbientAudio) 0.25 1.0 0.0 "Linear" :!class /Script/Engine.AudioComponent))

(event FortWeapon|EventOnLocalUnEquip
  ;[1 1 1] On Equip, turn on the Ambient Audio - On Unequip, turn it off.
  (|CallParentFunction)
  (Audio|Components|Audio|FadeOut (Variables|Default|GetAmbientAudio) 0.25 0.0 "Linear" :!class /Script/Engine.AudioComponent))

(fn ConstructionScript
  (|CallParentFunction))
