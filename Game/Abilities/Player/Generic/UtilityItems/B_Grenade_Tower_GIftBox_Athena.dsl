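; Gift Box variant of the player's tower grenade utility item that gives its held skeletal mesh a random colour.
; On BeginPlay it prints its display name for debugging, calls the parent, then has the authority roll Colour Random.
; The OnRep for that value calls UpdateMesh, which swaps between the default and green gift box meshes where cosmetic events can run.
; The weapon visibility event is an empty override.
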(event EventBeginPlay
  (Development|PrintString (Utilities|GetDisplayName self) true true "(R=0.000000,G=0.660000,B=1.000000,A=1.000000)" 1e+01)
  (|CallParentFunction)
  (Networking|SwitchHasAuthority
    (:Authority
      (CallFunction|RandomizeColour))))

(event EventOnWeaponVisibilityChanged
  (params (bVisible :type Boolean) (bSetForLocalControllerOnly :type Boolean)))

(fn OnRep_ColourRandom
  (CallFunction|UpdateMesh))

(fn RandomizeColour
  (Networking|SwitchHasAuthority
    (:Authority
      (Variables|Default|SetColour%20Random (Math|Random|RandomIntegerinRange 1 2)))))

(fn UpdateMesh
  (Networking|CanExecuteCosmeticEvents
    (:True
      (Components|SkinnedMesh|SetSkinnedAssetandUpdate (Variables|Default|GetSK_Gift_Box) (Utilities|Select :Index (Variables|Default|GetColour%20Random) :Option%200 "/Game/Athena/Items/Consumables/SuperTowerGrenade/Levels/GiftBox/BuildingPieces/SK_Gift_Box.SK_Gift_Box" :Option%201 "/Game/Athena/Items/Consumables/SuperTowerGrenade/Levels/GiftBox/BuildingPieces/SK_Gift_Box.SK_Gift_Box" :Option%202 "/Game/Athena/Items/Consumables/SuperTowerGrenade/Levels/GiftBox/BuildingPieces/SK_Gift_Box_Green.SK_Gift_Box_Green") true))))

(fn ConstructionScript
  (|CallParentFunction))
