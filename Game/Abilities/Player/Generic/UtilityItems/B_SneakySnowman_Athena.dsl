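; Sneaky Snowman utility item for Athena that fires its ability when the local player unequips it.
; OnLocalUnEquip caches the instigator pawn and, if ShouldSnowmanInTurret is positive, stops for pawns driving a mounted turret or Biplane (Ferret).
; Otherwise it sends the generic ActiveAbility.Activate gameplay event to the pawn.
; It then removes zero-count SnowmanItemDef instances from inventory, or calls RemoveItemFromPlayer if that definition is invalid.
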(event EventBeginPlay)

(event EventTick
  (params (DeltaSeconds :type Float%20%28single-precision%29)))

(event FortWeapon|EventOnLocalUnEquip
  (bind _playerpawn (Variables|Default|GetPlayerPawn))
  (bind _returnvalue (Vehicle|GetVehicleInterface _playerpawn))
  (bind _returnvalue_2 (Utilities|ToObject(Interface) _returnvalue))
  ;[1 1 1] EVENT -> OnUnequip
  (bind _asfort_player_pawn (Utilities|Casting|CastToFortPlayerPawn (Game|GetInstigator)))
  (Variables|Default|SetPlayerPawn _asfort_player_pawn)
  (if (> (ScalableFloat|GetValueatLevel (Variables|Default|GetShouldSnowmanInTurret)) 0.0)
    ;;[1 1 1] If they are in Mounted Turret or Biplane, dont do anything.
    (if (Vehicle|IsDrivingVehicle _playerpawn)
      (Utilities|Casting|CastToFortMountedTurret _returnvalue_2
        (:CastFailed
          (Utilities|Casting|CastToFortAthenaFerretVehicle _returnvalue_2
            (:CastFailed
              _))))
      (else
        _))
    (else
      _))
  (Ability|SendGameplayEventtoActor (Variables|Default|GetPlayerPawn) "(TagName=\"Abilities.Generic.ActiveAbility.Activate\")")
  (bind _asfort_player_controller (Utilities|Casting|CastToFortPlayerController (Pawn|GetController (Variables|Default|GetPlayerPawn))))
  (Utilities|IsValid (Variables|Default|GetSnowmanItemDef)
    (:Is%20Valid
      (Inventory|FindItemInstancesFromDefinition :self _asfort_player_controller :ItemDefinition (Variables|Default|GetSnowmanItemDef) :ItemArray (Variables|Default|GetSnowmanItemInstances))
      (for _array_element (Variables|Default|GetSnowmanItemInstances)
        (if (== (Item|GetNuminStack _array_element) 0)
          (Inventory|RemoveFortItemFromPlayer _asfort_player_controller _array_element -1 true))))
    (:Is%20Not%20Valid
      (Inventory|RemoveItemFromPlayer :PlayerController _asfort_player_controller :ItemDefinition (Variables|Default|GetSnowmanItemDef) :AmountToRemove -1 :bForceRemoval true :bSkipVerbRemoval false))))

(fn ConstructionScript
  (|CallParentFunction))
