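; Generic utility item Blueprint that manages HUD key-action hints for throwable consumables, gated by the UseThrowConsumableHudKeyActions flag.
; On BeginPlay it calls the parent and shows the first two key actions (indices 0 and 1).
; The Targeting custom event swaps the second action for the third while targeting and restores it afterward.
; The Visible custom event shows or hides all three actions together.
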(event EventBeginPlay
  (|CallParentFunction)
  (if (Variables|Default|GetUseThrowConsumableHudKeyActions)
    (WeaponHUDData|SetHudKeyActionsVisibility (Utilities|Array|MakeArray (Utilities|Struct|MakeWeaponHudKeyActionVisibility 0 true) (Utilities|Struct|MakeWeaponHudKeyActionVisibility 1 true)))))

(event Custom|ThrowConsumableHudActionKey_Targeting
  (params (Targeting :type Boolean))
  (if (Variables|Default|GetUseThrowConsumableHudKeyActions)
    (if Targeting
      (WeaponHUDData|SetHudKeyActionsVisibility (Utilities|Array|MakeArray (Utilities|Struct|MakeWeaponHudKeyActionVisibility 1) (Utilities|Struct|MakeWeaponHudKeyActionVisibility 2 true)))
      (else
        (WeaponHUDData|SetHudKeyActionsVisibility (Utilities|Array|MakeArray (Utilities|Struct|MakeWeaponHudKeyActionVisibility 1 true) (Utilities|Struct|MakeWeaponHudKeyActionVisibility 2)))))))

(event Custom|ThrowConsumableHudActionKey_Visible
  (params (Visible :type Boolean))
  (if (Variables|Default|GetUseThrowConsumableHudKeyActions)
    (WeaponHUDData|SetHudKeyActionsVisibility (Utilities|Array|MakeArray (Utilities|Struct|MakeWeaponHudKeyActionVisibility 0 Visible) (Utilities|Struct|MakeWeaponHudKeyActionVisibility 1 Visible) (Utilities|Struct|MakeWeaponHudKeyActionVisibility 2 Visible)))))

(fn ConstructionScript
  (|CallParentFunction))
